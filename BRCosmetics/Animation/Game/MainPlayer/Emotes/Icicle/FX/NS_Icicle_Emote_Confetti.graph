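# Niagara system that throws small confetti sprites for the Icicle player emote.
# The active Confetti001 emitter continuously spawns short-lived sprites at a randomized rate, with random size, rotation and sub-image, tinted by a time-of-day color.
# Particles launch with randomized cone velocity, slow under drag and gravity, and shrink along a curve late in life.
# A second, lower-rate Confetti emitter and the color-scaling modules are disabled.
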NiagaraSystem /BRCosmetics/Animation/Game/MainPlayer/Emotes/Icicle/FX/NS_Icicle_Emote_Confetti
dependencies {usedDataInterfaces [&NiagaraDataInterfaceCurve], usedDynamicInputs [], usedModules [&SystemState]}
emitters {
 Confetti {
  bEnabled=false
  emitterUpdateScript [
   &EmitterState()
   &SpawnRate(SpawnRate=&RandomRangeFloat(Minimum=30, Maximum=50), Use Spawn Probability=false)
  ]
  particleSpawnScript [
   /Niagara/Modules/Spawn/Initialization/InitializeParticle(
    Lifetime=&RandomRangeFloat(Minimum=0.2, Maximum=1)
    Module.Write Color=true
    Module.Write Lifetime=true
    Module.Write Mass=true
    Module.Write Material Random=true
    Module.Write Position=false
    Module.Write Scale=false
    Module.Write SpriteRotation=true
    Module.Write SpriteSize=true
    Position=&SimulationPosition()
    Sprite Rotation=&RandomRangeFloat(Minimum=0, Maximum=360)
    Sprite Size=&Vector2DFromFloat(Value=&RandomRangeFloat(Minimum=3, Maximum=6))
   ) {enabled=false}
   /Niagara/Modules/Spawn/Initialization/V2/InitializeParticle(
    Lifetime=&RandomRangeFloat(Minimum=0.2, Maximum=1)
    Mass Mode=::NewEnumerator1
    Sprite Rotation Angle=&RandomRangeFloat(Minimum=0, Maximum=360)
    Sprite Rotation Mode=::NewEnumerator0
    Sprite Size=&Vector2DFromFloat(Value=&RandomRangeFloat(Minimum=2, Maximum=5))
    Sprite Size Mode=::NewEnumerator4
    UsePositionOffset=false
    Write Material Random=true
   ) {module=InitializeParticle001}
   &AddVelocity(
    Cone Angle=90
    Module.UseConeFalloff=false
    Module.UseVelDistribution=false
    Override Seed=false
    Velocity Mode=::NewEnumerator2
    Velocity Speed=&RandomRangeFloat(Minimum=100, Maximum=225)
   )
  ]
  particleUpdateScript [
   &ParticleState()
   &Drag(Drag=&RandomRangeFloat(Minimum=5, Maximum=10), Use Linear Drag=true, Use Rotational Drag=false)
   &GravityForce(Gravity=&RandomRangeVector(Minimum=<0 0 -360>, Maximum=<1 1 -200>))
   &SolveForcesAndVelocity(Clamp Velocity=false, Limit Acceleration=false)
   &ScaleSpriteSize(Scale Factor=&Vector2DFromFloat(Value=&FloatFromCurve(FloatCurve=DataInterface() {propertyValues {Curve {keys [
          {arriveTangent=-0.4596, interpMode=RCIM_Cubic, leaveTangent=-0.4596, tangentMode=RCTM_User, time=0.55, value=1}
          {arriveTangent=-4.9031, interpMode=RCIM_Cubic, leaveTangent=-4.9031, tangentMode=RCTM_User, time=1, value=0}
         ]}}})))
   /Niagara/Modules/Update/Color/ScaleColorBySpeed() {enabled=false}
  ]
  rendererClasses [&NiagaraSpriteRendererProperties]
  renderers [{rendererClass=&NiagaraSpriteRendererProperties}]
 }
 Confetti001 {
  emitterUpdateScript [
   &EmitterState()
   &SpawnRate(SpawnRate=&RandomRangeFloat(Minimum=60, Maximum=100), Use Spawn Probability=false)
  ]
  particleSpawnScript [
   /Niagara/Modules/Spawn/Initialization/InitializeParticle(
    Lifetime=&RandomRangeFloat(Minimum=0.2, Maximum=1)
    Module.Write Color=true
    Module.Write Lifetime=true
    Module.Write Mass=true
    Module.Write Material Random=true
    Module.Write Position=false
    Module.Write Scale=false
    Module.Write SpriteRotation=true
    Module.Write SpriteSize=true
    Position=&SimulationPosition()
    Sprite Rotation=&RandomRangeFloat(Minimum=0, Maximum=360)
    Sprite Size=&Vector2DFromFloat(Value=&RandomRangeFloat(Minimum=3, Maximum=6))
   ) {enabled=false}
   /Niagara/Modules/Spawn/Initialization/V2/InitializeParticle(
    Color=&FNTimeOfDayColor(
     Night Color=<rgba 0.18 0.24 0.3 1>
     Morning Color=<rgba 0.3 0.2478 0.21 1>
     Day Color=<rgba 0.4 0.3536 0.32 1>
     Evening Color=<rgba 0.3 0.2391 0.195 1>
    )
    Lifetime=&RandomRangeFloat(Minimum=0.2, Maximum=1)
    Mass Mode=::NewEnumerator1
    Sprite Rotation Angle=&RandomRangeFloat(Minimum=0, Maximum=360)
    Sprite Rotation Mode=::NewEnumerator0
    Sprite Size=&Vector2DFromFloat(Value=&RandomRangeFloat(Minimum=2, Maximum=5))
    Sprite Size Mode=::NewEnumerator4
    UsePositionOffset=false
    Write Material Random=true
   ) {module=InitializeParticle001}
   &AddVelocity(
    Cone Angle=90
    Module.UseConeFalloff=false
    Module.UseVelDistribution=false
    Override Seed=false
    Velocity Mode=::NewEnumerator2
    Velocity Speed=&RandomRangeFloat(Minimum=100, Maximum=225)
   )
   /SetVariables(Particles.SubImageIndex=&RandomRangeFloat(Minimum=0, Maximum=4)) {isSetParameters=true}
  ]
  particleUpdateScript [
   &ParticleState()
   &Drag(Drag=&RandomRangeFloat(Minimum=5, Maximum=10), Use Linear Drag=true, Use Rotational Drag=false)
   &GravityForce(Gravity=&RandomRangeVector(Minimum=<0 0 -360>, Maximum=<1 1 -200>))
   &SolveForcesAndVelocity(Clamp Velocity=false, Limit Acceleration=false)
   &ScaleSpriteSize(Scale Factor=&Vector2DFromFloat(Value=&FloatFromCurve(FloatCurve=DataInterface() {propertyValues {Curve {keys [
          {arriveTangent=-0.4596, interpMode=RCIM_Cubic, leaveTangent=-0.4596, tangentMode=RCTM_User, time=0.55, value=1}
          {arriveTangent=-4.9031, interpMode=RCIM_Cubic, leaveTangent=-4.9031, tangentMode=RCTM_User, time=1, value=0}
         ]}}})))
   &ScaleColor(
    Scale RGB=&MakeVector(
     X=1
     Y=&Add_Float(
      A=&RandomRangeFloat(Minimum=0, Maximum=1000)
      B=&Lerp_Float(A=0, B=&RandomRangeFloat(Minimum=25, Maximum=150), Alpha=$Particles.NormalizedAge)
     )
     Z=1
    )
    ScaleA=false
    ScaleRGB=false
   ) {enabled=false}
   &ScaleColorBySpeed() {enabled=false}
  ]
  rendererClasses [&NiagaraSpriteRendererProperties]
  renderers [{rendererClass=&NiagaraSpriteRendererProperties}]
 }
}
summary {
 systemName=NS_Icicle_Emote_Confetti
 emitters [
  {bEnabled=false, emitterName=Confetti, rendererClasses [&NiagaraSpriteRendererProperties]}
  {emitterName=Confetti001, rendererClasses [&NiagaraSpriteRendererProperties]}
 ]
 userVariables []
}
