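# Niagara system for the Icicle emote on the main player, producing a continuous icy smoke effect from two sprite emitters.
# Idle_Flakes spawns small pale-blue specks outward from a sphere, moved by curl noise, gravity and drag as they spin, grow and fade.
# Idle_Puffs spawns larger, short-lived tinted puffs that sink, expand and animate through SubUV frames.
# A SkeletalMesh user parameter is exposed, but no listed module reads it.
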NiagaraSystem /BRCosmetics/Animation/Game/MainPlayer/Emotes/Icicle/FX/NS_Icicle_Emote_IceSmoke
dependencies {
 usedDataInterfaces [&NiagaraDataInterfaceCurve, &NiagaraDataInterfaceColorCurve]
 usedDynamicInputs [&SimulationPosition, &MakeVectorFromLinearColorRGB]
 usedModules [&SystemState]
}
emitters {
 Idle_Flakes {
  emitterUpdateScript [
   &EmitterState()
   &SpawnRate(SpawnRate=20, Use Spawn Probability=true)
  ]
  particleSpawnScript [
   &InitializeParticle(
    Lifetime=&RandomRangeFloat(Minimum=2, Maximum=3)
    Mass Mode=::NewEnumerator1
    Sprite Rotation Angle=&RandomRangeFloat(Minimum=0, Maximum=1)
    Sprite Rotation Mode=::NewEnumerator0
    Sprite Size=&Vector2DFromFloat(Value=&RandomRangeFloat(Minimum=2, Maximum=5))
    Sprite Size Mode=::NewEnumerator4
    UsePositionOffset=false
    Write Material Random=true
   ) {module=InitializeParticle001}
   &Color(Color=<rgba 0.4 0.59 1 1>, Scale Alpha=$Particles.MaterialRandom)
   &ShapeLocation(MaskSpawnGroup=false, Override Seed=false, Sphere Radius=1, UseOwnerScale=true)
   &AddVelocity(Velocity=&Multiply_VectorByFloat(Vector=$Particles.ShapeLocation.ShapeNormal, Float=&RandomRangeFloat(Minimum=25, Maximum=50)))
   &CameraOffset(Camera Offset Amount=-20)
  ]
  particleUpdateScript [
   &CurlNoiseForce(Noise Strength=20, Pan Noise=false)
   &GravityForce(Gravity=<0 0 -50>)
   &Drag(
    Drag=&RandomRangeFloat(Minimum=1, Maximum=2)
    Rotational Drag=20
    Use Linear Drag=true
    Use Rotational Drag=false
   )
   &SolveForcesAndVelocity(Clamp Velocity=false, Limit Acceleration=false)
   &ParticleState()
   &SpriteRotationRate(Rotation Rate=&FloatFromCurve(
     FloatCurve=DataInterface() {propertyValues {Curve {keys [<at 0 0 RCIM_Cubic>,<at 0.3 1 RCIM_Cubic>,<at 1 0>]}}}
     Scale Curve=&RandomRangeFloat(Minimum=-100, Maximum=100)
    ))
   &ScaleColor(
    Scale Alpha=&FloatFromCurve(FloatCurve=DataInterface() {propertyValues {Curve {keys [<at 0.6 1>,<at 1 0>]}}})
    ScaleA=true
    ScaleRGB=true
   )
   &ScaleSpriteSize(Scale Factor=&Vector2DFromFloat(Value=&FloatFromCurve(FloatCurve=DataInterface() {propertyValues {Curve {keys [
          <at 0 0 RCIM_Cubic>
          {arriveTangent=2, interpMode=RCIM_Cubic, leaveTangent=2, time=0.283, value=0.995}
          {arriveTangent=-1.3877, interpMode=RCIM_Cubic, leaveTangent=-1.3877, time=0.5, value=1}
          <at 1 0 RCIM_Cubic>
         ]}}})))
  ]
  rendererClasses [&NiagaraSpriteRendererProperties]
  renderers [{rendererClass=&NiagaraSpriteRendererProperties}]
 }
 Idle_Puffs {
  emitterUpdateScript [
   &EmitterState()
   &SpawnRate(SpawnRate=20, Use Spawn Probability=false)
  ]
  particleSpawnScript [
   &InitializeParticle(
    Color=<rgba 1 1 1 0.1>
    Lifetime=&RandomRangeFloat(Minimum=0.8, Maximum=1)
    Mass Mode=::NewEnumerator1
    Sprite Rotation Angle=&RandomRangeFloat(Minimum=0, Maximum=360)
    Sprite Rotation Mode=::NewEnumerator0
    Sprite Size=&Vector2DFromFloat(Value=&RandomRangeFloat(Minimum=10, Maximum=20))
    Sprite Size Mode=::NewEnumerator4
    UsePositionOffset=false
    Write Material Random=true
   ) {module=InitializeParticle001}
   &CameraOffset(Camera Offset Amount=-20) {enabled=false}
   &ShapeLocation(MaskSpawnGroup=false, Override Seed=false, Sphere Radius=10, UseOwnerScale=true)
  ]
  particleUpdateScript [
   &GravityForce(Gravity=<0 0 -50>)
   &Drag(
    Drag=&RandomRangeFloat(Minimum=5, Maximum=5)
    Rotational Drag=20
    Use Linear Drag=true
    Use Rotational Drag=false
   )
   &SolveForcesAndVelocity(Clamp Velocity=false, Limit Acceleration=false)
   &ParticleState()
   &SubUVAnimation(
    Sub UV Play Rate:NiagaraFloat=&Multiply_Float(A=0.3, B=$Particles.NormalizedAge)
    SubUV Animation Mode:ENiagara_SubUVLookupMode=::NewEnumerator3
   ) {module=SubUVAnimation001}
   &SpriteRotationRate(Rotation Rate=&FloatFromCurve(
     FloatCurve=DataInterface() {propertyValues {Curve {keys [<at 0 0 RCIM_Cubic>,<at 0.3 1 RCIM_Cubic>,<at 1 0>]}}}
     Scale Curve=&RandomRangeFloat(Minimum=-20, Maximum=20)
    ))
   &ScaleSpriteSize(Scale Factor=&Vector2DFromFloat(Value=&FloatFromCurve(
      FloatCurve=DataInterface() {propertyValues {Curve {keys [{arriveTangent=1.4732, interpMode=RCIM_Cubic, time=0, value=0.5}, <at 0.9951 1.6633 RCIM_Cubic>]}}}
      Scale Curve=2
     )))
   &ScaleColor(
    Scale Mode=::NewEnumerator1
    Scale RGBA=&MakeVector4_FromLinearColor(LinearColor=&ColorFromCurve(ColorCurve=DataInterface() {propertyValues {
        AlphaCurve {keys [
          <at 0 0 RCIM_Cubic>
          {arriveTangent=1.5193, interpMode=RCIM_Cubic, leaveTangent=1.5193, time=0.2235, value=0.9746}
          {arriveTangent=-1.2368, interpMode=RCIM_Cubic, leaveTangent=-1.2368, time=0.5352, value=0.8131}
          <at 1.0115 0>
          <at 1.094 0>
         ]}
        BlueCurve {keys [<at 0 1 RCIM_Cubic>,<at 1 1 RCIM_Cubic>]}
        GreenCurve {keys [<at 0 0.715 RCIM_Cubic>,<at 1 0.79 RCIM_Cubic>]}
        RedCurve {keys [<at 0 0.62 RCIM_Cubic>,<at 1 0.7 RCIM_Cubic>]}
       }}))
    ScaleRGBA=true
   ) {module=ScaleColor001}
  ]
  rendererClasses [&NiagaraSpriteRendererProperties]
  renderers [{rendererClass=&NiagaraSpriteRendererProperties}]
 }
}
summary {
 systemName=NS_Icicle_Emote_IceSmoke
 emitters [
  {emitterName=Idle_Flakes, rendererClasses [&NiagaraSpriteRendererProperties]}
  {emitterName=Idle_Puffs, rendererClasses [&NiagaraSpriteRendererProperties]}
 ]
 userVariables [User.SkeletalMesh {
   type=NiagaraDataInterfaceSkeletalMesh
   defaultValue {
    struct=NiagaraExt_VariableValue_DataInterface
    value {dataInterface=&self:NiagaraDataInterfaceSkeletalMesh_0}
   }
  }]
}
